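; Quest milestone ability that tracks how far the owning player pawn travels horizontally, presumably while in a vehicle, and reports it as quest progress.
; On activation it caches the pawn's location and starts a looping timer at a randomized interval, which adds whole meters moved to a running total.
; Past a threshold, the total is sent to the quest manager as a custom stat event and reset.
; Ending the ability clears the timer and sends any leftover distance above a smaller threshold.
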(event Ability|EventActivateAbilityFromEvent
  (params (EventData :type Gameplay%20Event%20Data%20Structure%20%28by%20ref%29))
  ;[1 1 1] Activate ability
  (|CallParentFunction EventData)
  (bind _output_get (Variables|GAVars|Data|SetPreviousLocation (Transformation|GetActorLocation (Variables|GAVars|Data|GetOwningPlayerPawn))))
  (Variables|GAVars|Data|SetCurrentLocation _output_get)
  (bind _output_get_2 (Variables|GAVars|SetInterval (Math|Random|RandomFloatinRange 1.7 2.5)))
  (bind _returnvalue (Utilities|Time|SetTimerbyEvent (delegate Custom|SendDistanceUpdate) _output_get_2 true))
  (Variables|GAVars|Data|SetTImerHandle _returnvalue))

(event Custom|SendDistanceUpdate
  (bind (_previouslocation_x _previouslocation_y) (Math|Vector|BreakVector (Variables|GetPreviousLocation_X)))
  (bind (_currentlocation_x _currentlocation_y) (Math|Vector|BreakVector (Variables|GetCurrentLocation_X)))
  ;[1 1 1] Send Distance Update
  (Variables|GAVars|Data|SetCurrentLocation (Transformation|GetActorLocation (Variables|GAVars|Data|GetOwningPlayerPawn)))
  ;;[1 1 1] Cache distance traveled (horizontal only)
  (Variables|GAVars|Data|SetMetersTraveled (+ (Math|Float|Floor (/ (Math|Vector|VectorLength (Math|Vector|vector-vector (Math|Vector|MakeVector _currentlocation_x _currentlocation_y) (Math|Vector|MakeVector _previouslocation_x _previouslocation_y))) 1e+02)) (Variables|GAVars|Data|GetMetersTraveled)))
  (Variables|GAVars|Data|SetPreviousLocation (Variables|GAVars|Data|GetCurrentLocation)) ; Reset location
  ;;[1 1 1] Send Quest Progress
  (if (> (Variables|GAVars|Data|GetMetersTraveled) 100)
    (Quest|SendComplexCustomStatEvent (Variables|GAVars|Data|GetOwningPlayerQuestManager) 0 (Variables|GAVars|Setup|GetT_Quest) "(GameplayTags=)" (Variables|GAVars|Data|GetMetersTraveled))
    (Variables|GAVars|Data|SetMetersTraveled 0)))

(event Ability|EventOnEndAbility
  (params (bWasCancelled :type Boolean))
  (bind _timerhandle (Variables|GAVars|Data|GetTImerHandle))
  ;[1 1 1] On End Ability
  (|CallParentFunction bWasCancelled)
  (if (Utilities|Time|IsValidTimerHandle _timerhandle)
    (Utilities|Time|ClearandInvalidateTimerbyHandle _timerhandle)
    (else
      _))
  (CallFunction|FInalDistanceUpdate))

(event Custom|FInal%20Distance%20Update
  ;[1 1 1] Check if player moved, if so send last update
  (if (>= (Variables|GAVars|Data|GetMetersTraveled) 10)
    ;[1 1 1] Send Distance Update
    ;;[1 1 1] Send Quest Progress
    (Quest|SendComplexCustomStatEvent (Variables|GAVars|Data|GetOwningPlayerQuestManager) 0 (Variables|GAVars|Setup|GetT_Quest) "(GameplayTags=)" (Variables|GAVars|Data|GetMetersTraveled))
    (Variables|GAVars|Data|SetMetersTraveled 0)))
